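# Widget Blueprint tile for a cosmetic variant redirector, built on the native FortVariantRedirectorTile class.
# It stacks an uppercase variant-name heading, a separator image and a wrapping dynamic entry box of AthenaVariantRedirectorTileButton options.
# The name text and entry box are inherited widgets, and the Blueprint adds layout and styling around them.
# GetDesiredFocusTarget returns GetButtonToFocus from the parent; the event graph holds only unconnected event stubs.
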WidgetBlueprint /MeshCosmetics/UI/AthenaVariantRedirectorTile
graphs [
 EventGraph {nodes [
   Event PreConstruct()
   Event Construct()
   Event Tick()
  ]}
 GetDesiredFocusTarget {nodes [
   GetDesiredFocusTarget() {is=title K2Node_FunctionEntry}
   > Return Node!(
    execute
    FocusTarget:Widget Object Reference=GetButtonToFocus!(self:Fort Variant Redirector Tile Object Reference) -> (ReturnValue:Widget Object Reference)
   )
  ]}
]
variables []
widgets {
 info {
  inheritedWidgetCount=2
  parentClass=/Script/MeshCosmeticsUI.FortVariantRedirectorTile
  rootWidgetClass=&CommonBorder
  widgetCount=5
 }
 widgets [
  CommonBorder_0 {widgetClassPath=&CommonBorder, widget=CommonBorder_0, properties {padding=<ltrb 0 0 0 0>}}
  VerticalBox_1 {
   widgetClassPath=&VerticalBox
   parent=CommonBorder_0
   widget=VerticalBox_1
   slot {class=BorderSlot, properties {padding=<ltrb 0 0 0 0>}}
  }
  Text_VariantName {
   widgetClassPath=&CommonTextBlock
   bInherited=true
   parent=VerticalBox_1
   widget=Text_VariantName
   properties {
    mobileFontSizeMultiplier=0.9
    text=Text Block
    textTransformPolicy=ToUpper
    font {
     bForceMonospaced=false
     bMaterialIsStencil=false
     fontMaterial=null
     fontObject=&HeadingNow
     letterSpacing=0
     monospacedWidth=1
     size=18
     skewAmount=0
     typefaceFontName=75 Medium Italic
     outlineSettings {
      bApplyOutlineToDropShadows=false
      bMiteredCorners=false
      bSeparateFillAlpha=false
      outlineMaterial=null
      outlineSize=0
      outlineColor=<rgba 0 0 0 1>
     }
    }
    shadowOffset=<0 0>
    strikeBrush {
     bIsDynamicallyLoaded=false
     drawAs=NoDrawType
     imageType=NoImage
     mirroring=NoMirror
     resourceName=None
     resourceObject=null
     tiling=NoTile
     imageSize=<32 32>
     margin=<ltrb 0 0 0 0>
     outlineSettings {
      bUseBrushTransparency=false
      roundingType=HalfHeightRadius
      width=0
      color {colorUseRule=UseColor_Specified, specifiedColor=<rgba 0 0 0 0>}
      cornerRadii=<0 0 0 0>
     }
     tintColor {colorUseRule=UseColor_Specified, specifiedColor=<rgba 1 1 1 1>}
     uVRegion {bIsValid=false, max=<0 0>, min=<0 0>}
    }
   }
   slot {class=VerticalBoxSlot, properties {padding=<ltrb 8 0 0 16>}}
  }
  Image {
   widgetClassPath=&Image
   parent=VerticalBox_1
   widget=Image
   properties {brush {
     bIsDynamicallyLoaded=false
     drawAs=Image
     imageType=NoImage
     mirroring=NoMirror
     resourceName=None
     resourceObject=/FortUILibrary/MaterialInstances/MI_UI_Customization_Separator
     tiling=NoTile
     imageSize=<16 2>
     margin=<ltrb 0 0 0 0>
     outlineSettings {
      bUseBrushTransparency=false
      roundingType=HalfHeightRadius
      width=0
      color {colorUseRule=UseColor_Specified, specifiedColor=<rgba 0 0 0 0>}
      cornerRadii=<0 0 0 0>
     }
     tintColor {colorUseRule=UseColor_Specified, specifiedColor=<rgba 1 1 1 1>}
     uVRegion {bIsValid=false, max=<0 0>, min=<0 0>}
    }}
   slot {class=VerticalBoxSlot, properties {verticalAlignment=VAlign_Top, padding=<ltrb 8 0 8 8>}}
  }
  EntryBox_VariantOptions {
   widgetClassPath=&DynamicEntryBox
   bInherited=true
   bIsVariable=true
   parent=VerticalBox_1
   widget=EntryBox_VariantOptions
   properties {
    entryBoxType=Wrap
    entryWidgetClass=&/AthenaVariantRedirectorTileButton.AthenaVariantRedirectorTileButton_C
    numDesignerPreviewEntries=5
    entrySpacing=<16 16>
   }
   slot {class=VerticalBoxSlot, properties {padding=<ltrb 8 8 8 8>}}
  }
 ]
}
