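# Reusable CommonButtonBase widget for the UIKit slider, showing a configurable icon inside a size box.
# PreConstruct applies the exposed brush, alignment, padding and minimum button size to the Icon and its Content box.
# On mobile, it also overrides the icon's desired size with MobileIconSize.
# The icon starts collapsed, and ShownIcon toggles it between Collapsed and SelfHitTestInvisible.
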WidgetBlueprint /FortUILibrary/UIKit/Slider/WBP_UIKit_SliderButton
graphs [
 EventGraph {nodes [
   Event PreConstruct()
   > SetBrush(self=Get Icon() -> (Icon:Image Object Reference), InBrush=Get Icon_Brush() -> (Icon_Brush:Slate Brush Structure))
   n1 = SlotAsSizeBoxSlot!(Widget:Widget Object Reference=Get Icon()) -> (ReturnValue:Size Box Slot Object Reference)
   > SetVerticalAlignment(
    self:Size Box Slot Object Reference=@n1
    InVerticalAlignment=Get IconVerticalAlignment() -> (IconVerticalAlignment:EVerticalAlignment Enum=VAlign_Fill)
   )
   > SetHorizontalAlignment(
    self:Size Box Slot Object Reference=@n1
    InHorizontalAlignment=Get IconHorizontalAlignment() -> (IconHorizontalAlignment:EHorizontalAlignment Enum=HAlign_Fill)
   )
   > SetPadding!(execute, self:Size Box Slot Object Reference=@n1, InPadding:Margin Structure=Get IconPadding() -> (IconPadding:Margin Structure)) -> (then)
   > SetMinDesiredHeight!(
    execute
    self:Size Box Object Reference=Get Content() -> (Content:Size Box Object Reference)
    InMinDesiredHeight:float=Get ButtonMinHeight() -> (ButtonMinHeight:float)
   ) -> (then)
   > SetMinDesiredWidth!(execute, self:Size Box Object Reference=Get Content(), InMinDesiredWidth:float=Get ButtonMinWidth() -> (ButtonMinWidth:float)) -> (then)
   > Branch(Condition=IsMobileGame())
   > SetDesiredSizeOverride!(
    execute
    self:Image Object Reference=Get Icon()
    DesiredSize:Vector 2D Structure=Get MobileIconSize() -> (MobileIconSize:Vector 2D Structure)
   ) -> (then)
  ]}
 ShownIcon {nodes [
   n1 = ShownIcon!() -> (then, IsShow:bool=~) {is=title K2Node_FunctionEntry}
   > SetVisibility(self=Get Icon(), InVisibility=Select(Option 0=Collapsed, Option 1=SelfHitTestInvisible, Index=@n1))
  ]}
]
variables [
 ButtonMinHeight {type=number, default=32}
 ButtonMinWidth {type=number, default=32}
 Icon_Brush {
  type=SlateBrush
  default {
   bIsDynamicallyLoaded=false
   drawAs=Image
   imageType=NoImage
   mirroring=NoMirror
   resourceName=None
   resourceObject=null
   tiling=NoTile
   imageSize=<32 32>
   margin=<ltrb 0 0 0 0>
   outlineSettings {
    bUseBrushTransparency=false
    roundingType=HalfHeightRadius
    width=0
    color {colorUseRule=UseColor_Specified, specifiedColor=<rgba 0 0 0 0>}
    cornerRadii=<0 0 0 1>
   }
   tintColor {colorUseRule=UseColor_Specified, specifiedColor=<rgba 1 1 1 1>}
   uVRegion {bIsValid=false, max=<0 0>, min=<0 0>}
  }
 }
 IconHorizontalAlignment {type=EHorizontalAlignment, default=HAlign_Fill}
 IconPadding {type=Margin, default=<ltrb 0 0 0 0>}
 IconVerticalAlignment {type=EVerticalAlignment, default=VAlign_Fill}
 MobileIconSize {type=Vector2D, default=<40 40>}
]
widgets {
 info {inheritedWidgetCount=1, parentClass=&CommonButtonBase, rootWidgetClass=&GridPanel, widgetCount=4}
 widgets [
  Root {widgetClassPath=&GridPanel, widget=Root, properties {columnFill [1], rowFill [1]}}
  Content {
   widgetClassPath=&SizeBox
   bIsVariable=true
   parent=Root
   widget=Content
   properties {bOverride_MinDesiredHeight=true, bOverride_MinDesiredWidth=true}
   slot {class=GridSlot}
  }
  Icon {
   widgetClassPath=&Image
   bIsVariable=true
   parent=Content
   widget=Icon
   properties {visibility=Collapsed}
   slot {class=SizeBoxSlot}
  }
  InputActionWidget {widgetClassPath=&CommonActionWidget, bInherited=true, widget=null}
 ]
}
