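# Slider row widget for one setting in the spray customizer UI, extending FortVariantSprayCustomizerSetting with a titled button.
# An invisible analog slider overlays a progress bar; value changes update the bar and show percent or degrees per bValueInDegrees.
# Constraint changes rescale the step size to the range, while chevron buttons nudge by a fixed 0.01.
# Hover plays a focus animation and shows the chevrons; focus-path entry gives the slider keyboard focus next tick.
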WidgetBlueprint /MeshCosmetics/UI/MutableProjectorSprayCustomizerSetting
animations [Anim_Focus {
  binding_count=3
  length_seconds=1
  object=&self:Anim_Focus
  sequence {bindings [
    Button_Decrease {
     id=#1
     possessed_class=WBP_SliderValueButton_C
     tracks [{
       type=MovieSceneColorTrack
       display_name=Color and Opacity
       sections [{
         type=MovieSceneColorSection
         end_frame=20
         start_frame=0
         channels [
          Color.R {type=MovieSceneScriptingFloatChannel, keys [[0,1],[20,0]]}
          Color.G {type=MovieSceneScriptingFloatChannel, keys [[0,1],[20,0]]}
          Color.B {type=MovieSceneScriptingFloatChannel, keys [[0,1],[20,0]]}
          Color.A {type=MovieSceneScriptingFloatChannel, const [1,1]}
         ]
        }]
      }]
    }
    Button_Increase {
     id=#2
     possessed_class=WBP_SliderValueButton_C
     tracks [{
       type=MovieSceneColorTrack
       display_name=Color and Opacity
       sections [{
         type=MovieSceneColorSection
         end_frame=20
         start_frame=0
         channels [
          Color.R {type=MovieSceneScriptingFloatChannel, keys [[0,1],[20,0]]}
          Color.G {type=MovieSceneScriptingFloatChannel, keys [[0,1],[20,0]]}
          Color.B {type=MovieSceneScriptingFloatChannel, keys [[0,1],[20,0]]}
          Color.A {type=MovieSceneScriptingFloatChannel, const [1,1]}
         ]
        }]
      }]
    }
    SliderProgressBar {
     id=#3
     possessed_class=WBP_UIKit_ProgressBar_C
     tracks [{
       type=MovieSceneColorTrack
       display_name=Color and Opacity
       sections [{
         type=MovieSceneColorSection
         end_frame=20
         start_frame=0
         channels [
          Color.R {type=MovieSceneScriptingFloatChannel, keys [[0,1],[20,0]]}
          Color.G {type=MovieSceneScriptingFloatChannel, keys [[0,1],[20,0]]}
          Color.B {type=MovieSceneScriptingFloatChannel, keys [[0,1],[20,0]]}
          Color.A {type=MovieSceneScriptingFloatChannel, const [1,1]}
         ]
        }]
      }]
    }
   ]}
 }]
graphs [
 EventGraph {nodes [
   "On Hovered (WBP_UIKit_Button_Generic)"!() -> (OutputDelegate, then, Button:Common Button Base Object Reference) {is=title K2Node_ComponentBoundEvent}
   > PlayAnimation(InAnimation=Get Anim_Focus() -> (Anim_Focus:Widget Animation Object Reference), PlaybackSpeed=3)
   > ShownIcon!(
    execute
    self:WBP UIKit Slider Button Object Reference=[
     Get Button_Decrease() -> (Button_Decrease:WBP UIKit Slider Button Object Reference)
     Get Button_Increase() -> (Button_Increase:WBP UIKit Slider Button Object Reference)
    ]
    IsShow:bool=true
   ) -> (then)
   "On Unhovered (WBP_UIKit_Button_Generic)"!() -> (OutputDelegate, then, Button:Common Button Base Object Reference) {is=title K2Node_ComponentBoundEvent}
   > PlayAnimation(InAnimation=Get Anim_Focus(), PlayMode=Reverse, PlaybackSpeed=3)
   > ShownIcon(self=[Get Button_Decrease(), Get Button_Increase()], IsShow=false)
   "On Clicked (Button_Decrease)"!() -> (OutputDelegate, then, Button:Common Button Base Object Reference) {is=title K2Node_ComponentBoundEvent}
   n1 = GetValue!(self:Slider Object Reference=Get Slider_Value() -> (Slider_Value:Analog Slider Object Reference)) -> (ReturnValue:float)
   > Branch(Condition="float > float"(A=@n1))
   > SetValue!(execute, self:Slider Object Reference=Get Slider_Value(), InValue:float=float - float(A=@n1, B=0.01)) -> (then)
   "On Clicked (Button_Increase)"!() -> (OutputDelegate, then, Button:Common Button Base Object Reference) {is=title K2Node_ComponentBoundEvent}
   n2 = GetValue(self=Get Slider_Value())
   > Branch(Condition="float < float"(A=@n2, B=1))
   > SetValue(self=Get Slider_Value(), InValue=float + float(A=@n2, B=0.01))
   Event PreConstruct()
   > "SetText (Text)"(self=Get Text_SettingName() -> (Text_SettingName:Common Text Object Reference), InText=Get SettingName() -> (SettingName:Text))
   > PlayAnimation(InAnimation=Get Anim_Focus(), PlayMode=Reverse, PlaybackSpeed=3)
   > ShownIcon(self=[Get Button_Decrease(), Get Button_Increase()], IsShow=false)
   n3 = Event On Change Value!() -> (OutputDelegate, then, Value:float, NormalizedValue:float) {is=title K2Node_Event}
   > UpdateSliderVisuals!(execute, self, SliderValue:float=@n3.Value, NormalizedSliderValue:double=@n3.NormalizedValue) -> (then)
   n4 = Event On Set Constraints!() -> (
    OutputDelegate
    then
    Constraints_LowerBound_Type:ERangeBoundTypes Enum
    Constraints_LowerBound_Value:float=~
    Constraints_UpperBound_Type:ERangeBoundTypes Enum
    Constraints_UpperBound_Value:float=~
   ) {is=title K2Node_Event}
   n5 = Get bValueInDegrees() -> (bValueInDegrees:bool)
   n6 = float - float(A=@n4.Constraints_UpperBound_Value, B=@n4.Constraints_LowerBound_Value)
   n7 = "float > float"(A=@n6)
   > SetStepSize!(
    execute
    self:Slider Object Reference=Get Slider_Value()
    InValue:float=float / float(A=SelectFloat(A=1, B=0.01, bPickA=@n5), B=SelectFloat(A=@n6, B=1, bPickA=@n7))
   ) -> (then)
   Event OnInitialized()
   > n8 = Cast To WBP_UIKit_BlockBackground_Tab_Enhanced!(
    execute
    Object=Get Block_Background!(self:WBP UIKit Button Generic Base Object Reference=Get WBP_UIKit_Button_Generic() -> (WBP_UIKit_Button_Generic:WBP UIKit Button Generic Object Reference)) -> (Block_Background:WBP UIKit Block Background Base Object Reference)
   ) -> (then, CastFailed, AsWBP UIKit Block Background Tab Enhanced:WBP UIKit Block Background Tab Enhanced Object Reference) {is=title K2Node_DynamicCast}
   > SetGridSlotPadding!(
    execute
    Widget:Widget Object Reference=Get FocusOutline!(self:WBP UIKit Block Background Tab Enhanced Object Reference=@n8) -> (FocusOutline:Image Object Reference)
    Padding_Left:float=-18
    Padding_Top:float=-18
    Padding_Right:float=-18
    Padding_Bottom:float=-18
   ) -> (then)
   Event OnAddedToFocusPath!() -> (OutputDelegate, then, InFocusEvent:Focus Event Structure) {is=title K2Node_Event}
   > DelayUntilNextTick!(execute) -> (then)
   > SetKeyboardFocus(self:Widget Object Reference=Get Slider_Value())
   Event OnRemovedFromFocusPath!() -> (OutputDelegate, then, InFocusEvent:Focus Event Structure) {is=title K2Node_Event}
   > DelayUntilNextTick()
  ]}
 UpdateSliderVisuals {nodes [
   n1 = UpdateSliderVisuals!() -> (then, SliderValue:float=~, NormalizedSliderValue:double=~) {is=title K2Node_FunctionEntry}
   > SetProgress!(
    execute
    self:WBP UIKit Progress Bar Base Object Reference=Get SliderProgressBar() -> (SliderProgressBar:WBP UIKit Progress Bar Object Reference)
    Progress:double=@n1.NormalizedSliderValue
   ) -> (then)
   n2 = Get SliderValue() -> (SliderValue:float)
   n3 = "To Text (Integer)"(Value=Round(A=Get SliderValue()) {is=title K2Node_CallFunction})
   > SetText!(
    execute
    self:WBP UIKit Button Generic Base Object Reference=Get WBP_UIKit_Button_Generic()
    Text:Text=Select(
     Option 0:Text=As Percent!(
      Value:float=@n2
      RoundingMode:ERoundingMode Enum=HalfFromZero
      bAlwaysSign:bool=false
      bUseGrouping:bool=true
      MinimumIntegralDigits:int=1
      MaximumIntegralDigits:int=324
      MinimumFractionalDigits:int=0
      MaximumFractionalDigits:int=0
     ) -> (ReturnValue:Text)
     Option 1:Text=Format Text!(Format="LOCTEXT(\"{0}º\")", 0:Text=@n3) -> (Result)
     Index=Get bValueInDegrees()
    ) -> (ReturnValue:Text)
   ) -> (then)
  ]}
]
variables [
 bValueInDegrees {type=boolean}
 Current Input Type {type=ECommonInputType, default=MouseAndKeyboard}
 SettingName {type=string, default="[PH] Setting Name"}
]
widgets {
 info {
  inheritedWidgetCount=1
  namedSlotCount=1
  parentClass=/Script/MeshCosmeticsUI.FortVariantSprayCustomizerSetting
  rootWidgetClass=&CommonBorder
  widgetCount=11
 }
 widgets [
  CommonBorder_1 {widgetClassPath=&CommonBorder, widget=CommonBorder_1, properties {padding=<ltrb 8 8 8 8>}}
  VerticalBox_0 {
   widgetClassPath=&VerticalBox
   parent=CommonBorder_1
   widget=VerticalBox_0
   slot {class=BorderSlot, properties {padding=<ltrb 8 8 8 8>}}
  }
  Text_SettingName {
   widgetClassPath=&CommonTextBlock
   bIsVariable=true
   parent=VerticalBox_0
   widget=Text_SettingName
   properties {
    mobileFontSizeMultiplier=1.4
    text=Text Block
    font {
     bForceMonospaced=false
     bMaterialIsStencil=false
     fontMaterial=null
     fontObject=&HeadingNow
     letterSpacing=0
     monospacedWidth=1
     size=18
     skewAmount=0
     typefaceFontName=75 Medium Italic
     outlineSettings {
      bApplyOutlineToDropShadows=false
      bMiteredCorners=false
      bSeparateFillAlpha=false
      outlineMaterial=null
      outlineSize=0
      outlineColor=<rgba 0 0 0 1>
     }
    }
    shadowOffset=<0 0>
   }
   slot {class=VerticalBoxSlot, properties {padding=<ltrb 8 0 0 16>}}
  }
  Image_Separator {
   widgetClassPath=&Image
   parent=VerticalBox_0
   widget=Image_Separator
   properties {brush {
     bIsDynamicallyLoaded=false
     drawAs=Image
     imageType=NoImage
     mirroring=NoMirror
     resourceName=None
     resourceObject=/FortUILibrary/MaterialInstances/MI_UI_Customization_Separator
     tiling=NoTile
     imageSize=<16 2>
     margin=<ltrb 0 0 0 0>
     outlineSettings {
      bUseBrushTransparency=false
      roundingType=HalfHeightRadius
      width=0
      color {colorUseRule=UseColor_Specified, specifiedColor=<rgba 0 0 0 0>}
      cornerRadii=<0 0 0 0>
     }
     tintColor {colorUseRule=UseColor_Specified, specifiedColor=<rgba 1 1 1 1>}
     uVRegion {bIsValid=false, max=<0 0>, min=<0 0>}
    }}
   slot {class=VerticalBoxSlot, properties {verticalAlignment=VAlign_Top, padding=<ltrb 8 0 8 26>}}
  }
  WBP_UIKit_Button_Generic {
   widgetClassPath=&WBP_UIKit_Button_Generic
   bIsVariable=true
   parent=VerticalBox_0
   widget=WBP_UIKit_Button_Generic
   properties {
    bHideInputAction=true
    bHideInputActionWithKeyboard=true
    backgroundClass=/FortUILibrary/UIKit/Buttons/Blocks/WBP_UIKit_BlockBackground_Tab_Enhanced.WBP_UIKit_BlockBackground_Tab_Enhanced_C
    backgroundMaterial=/FortUILibrary/UIKit/Buttons/MaterialInstances/CTAButtons/MI_UIKit_BgTab_Enhanced
    contentVerticalAlignment=VAlign_Top
    minHeight=88
    mobileMinHeight=88
    outlineMaterial=/FortUILibrary/UIKit/Buttons/MaterialInstances/CTAButtons/MI_UIKit_ButtonTabOutline_Enhanced
    text=75
    textTransformPolicy=None
    useIconBlock=false
    buttonPadding=<ltrb 0 0 0 0>
    font {
     bForceMonospaced=false
     bMaterialIsStencil=false
     fontMaterial=/FortUILibrary/UIKit/Buttons/MaterialInstances/GenericButtons/MI_UIKit_GenericText_Quiet
     fontObject=&HeadingNow
     letterSpacing=48
     monospacedWidth=1
     size=18
     skewAmount=0
     typefaceFontName=74 Regular
     outlineSettings {
      bApplyOutlineToDropShadows=false
      bMiteredCorners=false
      bSeparateFillAlpha=false
      outlineMaterial=null
      outlineSize=0
      outlineColor=<rgba 0 0 0 1>
     }
    }
    textPadding=<ltrb 0 11 0 0>
   }
   slot {class=VerticalBoxSlot}
  }
  HorizontalBox_2 {widgetClassPath=&HorizontalBox, namedSlotHost=WBP_UIKit_Button_Generic, widget=HorizontalBox_2}
  Button_Decrease {
   widgetClassPath=/FortUILibrary/UIKit/Slider/WBP_UIKit_SliderButton
   bIsVariable=true
   parent=HorizontalBox_2
   widget=Button_Decrease
   properties {
    holdData=&CommonUI_HoldData_Base_C
    iconHorizontalAlignment=HAlign_Center
    iconVerticalAlignment=VAlign_Top
    minWidth=58
    colorAndOpacity=<rgba 0 0 0 1>
    iconPadding=<ltrb 0 10 0 0>
    icon_Brush {
     bIsDynamicallyLoaded=false
     drawAs=Image
     imageType=FullColor
     mirroring=NoMirror
     resourceName=None
     resourceObject=/FortUILibrary/Textures/Icons/ToBeDeprecated/T_UI_Icon_Global_Chevron_Left
     tiling=NoTile
     imageSize=<32 32>
     margin=<ltrb 0 0 0 0>
     outlineSettings {
      bUseBrushTransparency=false
      roundingType=HalfHeightRadius
      width=0
      color {colorUseRule=UseColor_Specified, specifiedColor=<rgba 0 0 0 0>}
      cornerRadii=<0 0 0 1>
     }
     tintColor {colorUseRule=UseColor_Specified, specifiedColor=<rgba 1 1 1 1>}
     uVRegion {bIsValid=false, max=<0 0>, min=<0 0>}
    }
   }
   slot {class=HorizontalBoxSlot}
  }
  Overlay_73 {
   widgetClassPath=&Overlay
   parent=HorizontalBox_2
   widget=Overlay_73
   slot {class=HorizontalBoxSlot, properties {size {sizeRule=Fill, value=1}}}
  }
  SliderProgressBar {
   widgetClassPath=/FortUILibrary/UIKit/ProgressBar/WBP_UIKit_ProgressBar
   bIsVariable=true
   parent=Overlay_73
   widget=SliderProgressBar
   properties {height=8, useCircular=false, width=0, emptyColor=<rgba 1 1 1 0.2>}
   slot {class=OverlaySlot, properties {horizontalAlignment=HAlign_Fill, verticalAlignment=VAlign_Bottom, padding=<ltrb 0 0 0 12>}}
  }
  Slider_Value {
   widgetClassPath=/Script/CommonUI.AnalogSlider
   bInherited=true
   bIsVariable=true
   parent=Overlay_73
   widget=Slider_Value
   properties {sliderBarColor=<rgba 1 1 1 0>, sliderHandleColor=<rgba 1 1 1 0>}
   slot {class=OverlaySlot, properties {horizontalAlignment=HAlign_Fill, verticalAlignment=VAlign_Fill}}
  }
  Button_Increase {
   widgetClassPath=&WBP_UIKit_SliderButton
   bIsVariable=true
   parent=HorizontalBox_2
   widget=Button_Increase
   properties {
    holdData=&CommonUI_HoldData_Base_C
    iconHorizontalAlignment=HAlign_Center
    iconVerticalAlignment=VAlign_Top
    minWidth=58
    colorAndOpacity=<rgba 0 0 0 1>
    iconPadding=<ltrb 0 10 0 0>
    icon_Brush {
     bIsDynamicallyLoaded=false
     drawAs=Image
     imageType=FullColor
     mirroring=NoMirror
     resourceName=None
     resourceObject=/FortUILibrary/Textures/Icons/ToBeDeprecated/T_UI_Icon_Global_Chevron_Right
     tiling=NoTile
     imageSize=<32 32>
     margin=<ltrb 0 0 0 0>
     outlineSettings {
      bUseBrushTransparency=false
      roundingType=HalfHeightRadius
      width=0
      color {colorUseRule=UseColor_Specified, specifiedColor=<rgba 0 0 0 0>}
      cornerRadii=<0 0 0 1>
     }
     tintColor {colorUseRule=UseColor_Specified, specifiedColor=<rgba 1 1 1 1>}
     uVRegion {bIsValid=false, max=<0 0>, min=<0 0>}
    }
   }
   slot {class=HorizontalBoxSlot}
  }
 ]
}
