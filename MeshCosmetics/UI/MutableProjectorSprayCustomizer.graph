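# Locker UI screen, built on FortVariantSprayCustomizer, for customizing a spray projected onto a cosmetic, with an item preview.
# A Sprays tab holds a texture picker and an Adjustments tab holds Saturation, Wear and Scale settings.
# The graph updates subheader and tab icons, shows picker conflict warnings, and hides the purchase disclaimer in locker mode.
# Input method changes animate spacers around the tab prompts for gamepad.
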WidgetBlueprint /MeshCosmetics/UI/MutableProjectorSprayCustomizer
animations [Anim_PromptSpaceChange {
  binding_count=2
  length_seconds=0.2
  object=&self:Anim_PromptSpaceChange
  sequence {bindings [
    Spacer_205 {
     id=#1
     possessed_class=Spacer
     tracks [{
       type=MovieSceneDoubleVectorTrack
       display_name=Size
       sections [{
         type=MovieSceneDoubleVectorSection
         end_frame=4
         start_frame=0
         channels [
          Vector.X {type=MovieSceneScriptingDoubleChannel, keys [[0,0],[4,38]]}
          Vector.Y {type=MovieSceneScriptingDoubleChannel, const [0,1]}
         ]
        }]
      }]
    }
    Spacer {
     id=#2
     possessed_class=Spacer
     tracks [{
       type=MovieSceneDoubleVectorTrack
       display_name=Size
       sections [{
         type=MovieSceneDoubleVectorSection
         end_frame=4
         start_frame=0
         channels [
          Vector.X {type=MovieSceneScriptingDoubleChannel, keys [[0,0],[4,38]]}
          Vector.Y {type=MovieSceneScriptingDoubleChannel, const [0,1]}
         ]
        }]
      }]
    }
   ]}
 }]
graphs [
 BP_GetDesiredInputConfig {nodes [
   Get Desired Input Config() {is=title K2Node_FunctionEntry}
   > Return Node!(
    execute
    ReturnValue:UIInput Config Structure=Make UIInput Config!(
     bIgnoreMoveInput:bool=False
     bIgnoreLookInput:bool=False
     InputMode:ECommonInputMode Enum=All
     MouseCaptureMode:EMouseCaptureMode Enum=CaptureDuringMouseDown
     MouseLockMode:EMouseLockMode Enum=DoNotLock
     bHideCursorDuringViewportCapture:bool=True
    ) -> (UIInputConfig:UIInput Config Structure) {is=title K2Node_MakeStruct}
   )
  ]}
 EventGraph {nodes [
   n1 = "On Selected Changed (Button_SpraysTab)"!() -> (OutputDelegate, then, Button:Common Button Base Object Reference, Selected:bool) {is=title K2Node_ComponentBoundEvent}
   > Branch(Condition=@n1.Selected)
   > SetSubHeader!(
    execute
    self:WBP MPLocker Heading Object Reference=Get Heading() -> (Heading:WBP MPLocker Heading Object Reference)
    Text:Text="LOCTEXT(\"SPRAYS\")"
   ) -> (then)
   n2 = "On Conflict Status Updated (TexturePicker_Sprays)"!() -> (OutputDelegate, then, IsConflicting:bool, ConflictReason:Text) {is=title K2Node_ComponentBoundEvent}
   n3 = Get WarningMessage() -> (WarningMessage:WBP UIKit Info Box Object Reference)
   > SetVisibility(self=@n3, InVisibility=Select(Option 0=Collapsed, Option 1=HitTestInvisible, Index=@n2.IsConflicting))
   > SetText(self:WBP UIKit Info Box Object Reference=@n3, InText:Text=@n2.ConflictReason)
   n4 = "On Selected Changed (Button_SettingsTab)"!() -> (OutputDelegate, then, Button:Common Button Base Object Reference, Selected:bool) {is=title K2Node_ComponentBoundEvent}
   > Branch(Condition=@n4.Selected)
   > SetSubHeader(self=Get Heading(), Text="LOCTEXT(\"ADJUSTMENTS\")")
   n5 = OnInputChanged!() -> (OutputDelegate, then, bNewInputType:ECommonInputType Enum) {is=title K2Node_CustomEvent}
   n6 = "Equal (Enum)"(A=@n5.bNewInputType, B=Gamepad)
   n7 = Get CurrentInput Type() -> (CurrentInput Type:ECommonInputType Enum)
   > n8 = Branch(Condition=OR Boolean(A=@n6, B="Equal (Enum)"(A=@n7, B=Gamepad)))
   > n9 = PlayAnimation(
    InAnimation=Get Anim_PromptSpaceChange() -> (Anim_PromptSpaceChange:Widget Animation Object Reference)
    PlayMode=Select(Option 0:EUMGSequencePlayMode Enum=Reverse, Option 1:EUMGSequencePlayMode Enum, Index=@n6) -> (ReturnValue:EUMGSequencePlayMode Enum)
   )
   Set CurrentInput Type(execute=[@n9.then, @n8.else], CurrentInput Type:ECommonInputType Enum=@n5.bNewInputType) -> (Output_Get:ECommonInputType Enum)
   n10 = UpdatePreview!() -> (OutputDelegate, then, PreviewItem:Fort Item VM Object Reference) {is=title K2Node_CustomEvent}
   > DelayUntilNextTick!(execute) -> (then)
   > SetCurrentPreviewItem!(
    execute
    self:Fort Item Preview Widget Object Reference=Get FortItemPreviewWidget() -> (FortItemPreviewWidget:WBP Item Preview Object Reference)
    ItemVM:Fort Item VM Object Reference=@n10.PreviewItem
   ) -> (then)
   n11 = Event PreConstruct()
   > n12 = Branch(Condition=@n11.IsDesignTime)
   n13 = Get Heading()
   SetHeader!(execute=@n12.else, self:WBP MPLocker Heading Object Reference=@n13, Text:Text="LOCTEXT(\"ITEM PREVIEW\")") -> (then)
   > SetSubHeader(self=@n13, Text="LOCTEXT(\"SPRAYS\")")
   > n14 = Sequence()
   > n15 = Cast To VariantFilterTab!(execute, Object=Get Button_SpraysTab() -> (Button_SpraysTab:Common Button Base Object Reference)) -> (then, CastFailed, AsVariant Filter Tab:Variant Filter Tab Object Reference) {is=title K2Node_DynamicCast}
   > n16 = GetContainedImages!(execute, InWidget:Widget Object Reference=@n15) -> (then, ReturnValue:Array of Image Object References)
   > n17 = For Each Loop(Array:Array of Image Object References=@n16) -> (Array Element:Image Object Reference)
   > n18 = Cast To FortLazyImage!(execute, Object=@n17) -> (then, CastFailed, AsFort Lazy Image:Fort Lazy Image Object Reference) {is=title K2Node_DynamicCast}
   > SetBrushFromTexture!(
    execute
    self:Image Object Reference=@n18
    Texture:Texture 2D Object Reference=/Game/UI/Foundation/Textures/Icons/Locker/T_Icon_Spray_128
    bMatchSize:bool=false
   ) -> (then)
   n19 = Cast To VariantFilterTab(execute=@n14.then_1, Object=Get Button_SettingsTab() -> (Button_SettingsTab:Common Button Base Object Reference))
   > n20 = GetContainedImages(InWidget=@n19)
   > n21 = For Each Loop(Array=@n20)
   > n22 = Cast To FortLazyImage(Object=@n21)
   > SetBrushFromTexture(self=@n22, Texture=/Game/UI/Foundation/Textures/Icons/Locker/T_Icon_Spray_Adjustments_Tab_128, bMatchSize=false)
   n23 = Event Set Channel Slot Icon!() -> (OutputDelegate, then, ChannelSlotIcon:"Soft Object Reference (by ref)") {is=title K2Node_Event}
   > n24 = Cast To VariantFilterTab(Object=Get Button_SpraysTab())
   > n25 = GetContainedImages(InWidget=@n24)
   > n26 = For Each Loop(Array=@n25)
   > n27 = Cast To FortLazyImage(Object=@n26)
   > SetBrushFromLazyDisplayAsset!(
    execute
    self:Common Lazy Image Object Reference=@n27
    LazyObject:"Soft Object Reference (by ref)"=@n23.ChannelSlotIcon
    bMatchTextureSize:bool=false
   ) -> (then)
   n28 = Event Set Variant Editor Mode!() -> (OutputDelegate, then, bIsLockerMode:bool) {is=title K2Node_Event}
   > SetVisibility(
    self=Get Disclaimer_Container() -> (Disclaimer_Container:Horizontal Box Object Reference)
    InVisibility=Select(Option 0:ESlateVisibility Enum=HitTestInvisible, Option 1:ESlateVisibility Enum=Collapsed, Index=@n28.bIsLockerMode) -> (ReturnValue:ESlateVisibility Enum)
   )
   Event On Deactivated()
   n29 = Get FortItemPreviewWidget()
   > DeactivateWidget(self=@n29)
   > ClearPreviewItem(self:Fort Item Preview Widget Object Reference=@n29)
   Event On Activated()
   > ActivateWidget(self=@n29)
   Event Construct()
   > SetIsSelected!(execute, self:Common Button Base Object Reference=Get Button_SpraysTab(), InSelected:bool=true, bGiveClickFeedback:bool=true) -> (then)
   > SetIsSelected(self=Get Button_SettingsTab(), InSelected=true, bGiveClickFeedback=true)
   > Bind Event to On Input Method Changed(self=Get CommonInputSubsystem(), Delegate=Create Event())
   n30 = GetCurrentInputType(self=Get CommonInputSubsystem())
   > Set CurrentInput Type(CurrentInput Type=@n30)
   > Branch(Condition="Equal (Enum)"(A=@n30, B=Gamepad))
   > PlayAnimation(InAnimation=Get Anim_PromptSpaceChange())
  ]}
]
variables [CurrentInput Type {type=ECommonInputType, default=MouseAndKeyboard}]
widgets {
 info {
  inheritedWidgetCount=23
  parentClass=/Script/MeshCosmeticsUI.FortVariantSprayCustomizer
  rootWidgetClass=&SafeZone
  widgetCount=50
 }
 widgets [
  SafeZone {widgetClassPath=&SafeZone, bIsVariable=true, widget=SafeZone, properties {padBottom=false, padTop=false}}
  SizeBox_0 {
   widgetClassPath=&SizeBox
   bIsVariable=true
   parent=SafeZone
   widget=SizeBox_0
   properties {bOverride_WidthOverride=true, widthOverride=1920}
   slot {class=SafeZoneSlot, properties {hAlign=HAlign_Center}}
  }
  GridPanel {
   widgetClassPath=&GridPanel
   parent=SizeBox_0
   widget=GridPanel
   properties {columnFill [1,0], rowFill [1,0]}
   slot {class=SizeBoxSlot}
  }
  Locker {
   widgetClassPath=&GridPanel
   parent=GridPanel
   widget=Locker
   properties {columnFill [1,1,2], rowFill [0,0,1,0]}
   slot {class=GridSlot, properties {layer=1, padding=<ltrb 64 64 64 64>}}
  }
  VBox_MainStack {widgetClassPath=&VerticalBox, parent=Locker, widget=VBox_MainStack, slot {class=GridSlot, properties {column=2, rowSpan=4}}}
  HorizontalBox_1 {
   widgetClassPath=&HorizontalBox
   parent=VBox_MainStack
   widget=HorizontalBox_1
   slot {class=VerticalBoxSlot, properties {padding=<ltrb 16 0 0 14>}}
  }
  Heading {
   widgetClassPath=/FortniteCoreUI/Customization/MPL/WBP_MPLocker_Heading
   bIsVariable=true
   parent=HorizontalBox_1
   widget=Heading
   slot {class=HorizontalBoxSlot, properties {size {sizeRule=Fill, value=1}}}
  }
  VisibilityBorder_Touch {
   widgetClassPath=&CommonHardwareVisibilityBorder
   parent=HorizontalBox_1
   widget=VisibilityBorder_Touch
   properties {
    padding=<ltrb 20 -24 0 0>
    visibilityQuery {
     autoDescription=" ANY( Input.Touch )"
     tokenStreamVersion=0
     userDescription=""
     queryTokenStream [0,1,1,1,0]
     tagDictionary [{tagName=Input.Touch}]
    }
   }
   slot {class=HorizontalBoxSlot, properties {horizontalAlignment=HAlign_Right, verticalAlignment=VAlign_Center}}
  }
  Button_TouchClose {
   widgetClassPath=/FortUILibrary/Foundation/Buttons/CloseButton
   bInherited=true
   bIsVariable=true
   parent=VisibilityBorder_Touch
   widget=Button_TouchClose
   properties {
    onButtonBaseClicked=(&self:WidgetTree.Button_TouchClose.HandleOnButtonClicked)
    onButtonBaseDoubleClicked=(&self:WidgetTree.Button_TouchClose.HandleOnButtonDoubleClicked)
    onButtonBaseHovered=(&self:WidgetTree.Button_TouchClose.HandleOnButtonHovered)
    onButtonBaseUnhovered=(&self:WidgetTree.Button_TouchClose.HandleOnButtonUnhovered)
    onSelectedChangedBase=(&self:WidgetTree.Button_TouchClose.HandleOnSelectedChanged)
   }
   slot {class=BorderSlot, properties {padding=<ltrb 20 -24 0 0>}}
  }
  HBox_PrimaryTabNav {
   widgetClassPath=&HorizontalBox
   parent=VBox_MainStack
   widget=HBox_PrimaryTabNav
   slot {class=VerticalBoxSlot, properties {padding=<ltrb 16 0 0 16>}}
  }
  ActionWidget_TabLeft {
   widgetClassPath=&CommonActionWidget
   bInherited=true
   bIsVariable=true
   parent=HBox_PrimaryTabNav
   widget=ActionWidget_TabLeft
   properties {inputActions [{dataTable=&InputActionDataTable, rowName=SecondaryTabLeft}]}
   slot {class=HorizontalBoxSlot, properties {padding=<ltrb 0 0 8 0>}}
  }
  Spacer_205 {widgetClassPath=&Spacer, parent=HBox_PrimaryTabNav, widget=Spacer_205, properties {size=<0 0>}, slot {class=HorizontalBoxSlot}}
  Overlay_0 {widgetClassPath=&Overlay, parent=HBox_PrimaryTabNav, widget=Overlay_0, slot {class=HorizontalBoxSlot}}
  UIKit_Backplate {
   widgetClassPath=&WBP_UIKit_Backplate
   parent=Overlay_0
   widget=UIKit_Backplate
   properties {brightness=Dark, cornerRadius=Interactive Component}
   slot {class=OverlaySlot, properties {horizontalAlignment=HAlign_Fill, verticalAlignment=VAlign_Fill}}
  }
  HorizontalBox_0 {
   widgetClassPath=&HorizontalBox
   parent=Overlay_0
   widget=HorizontalBox_0
   slot {class=OverlaySlot, properties {padding=<ltrb 8 8 8 8>}}
  }
  Button_SpraysTab {
   widgetClassPath=/FortniteCoreUI/Customization/VariantFilterTab
   bInherited=true
   bIsVariable=true
   parent=HorizontalBox_0
   widget=Button_SpraysTab
   properties {
    onButtonBaseClicked=(&self:WidgetTree.Button_SpraysTab.HandleOnButtonClicked)
    onButtonBaseDoubleClicked=(&self:WidgetTree.Button_SpraysTab.HandleOnButtonDoubleClicked)
    onButtonBaseHovered=(&self:WidgetTree.Button_SpraysTab.HandleOnButtonHovered)
    onButtonBaseUnhovered=(&self:WidgetTree.Button_SpraysTab.HandleOnButtonUnhovered)
    onSelectedChangedBase=(&self:WidgetTree.Button_SpraysTab.HandleOnSelectedChanged)
   }
   slot {class=HorizontalBoxSlot}
  }
  Button_SettingsTab {
   widgetClassPath=&VariantFilterTab
   bInherited=true
   bIsVariable=true
   parent=HorizontalBox_0
   widget=Button_SettingsTab
   properties {
    onButtonBaseClicked=(&self:WidgetTree.Button_SettingsTab.HandleOnButtonClicked)
    onButtonBaseDoubleClicked=(&self:WidgetTree.Button_SettingsTab.HandleOnButtonDoubleClicked)
    onButtonBaseHovered=(&self:WidgetTree.Button_SettingsTab.HandleOnButtonHovered)
    onButtonBaseUnhovered=(&self:WidgetTree.Button_SettingsTab.HandleOnButtonUnhovered)
    onSelectedChangedBase=(&self:WidgetTree.Button_SettingsTab.HandleOnSelectedChanged)
   }
   slot {class=HorizontalBoxSlot}
  }
  Spacer {widgetClassPath=&Spacer, parent=HBox_PrimaryTabNav, widget=Spacer, properties {size=<0 0>}, slot {class=HorizontalBoxSlot}}
  ActionWidget_TabRight {
   widgetClassPath=&CommonActionWidget
   bInherited=true
   bIsVariable=true
   parent=HBox_PrimaryTabNav
   widget=ActionWidget_TabRight
   properties {inputActions [{dataTable=&InputActionDataTable, rowName=SecondaryTabRight}]}
   slot {class=HorizontalBoxSlot, properties {padding=<ltrb 8 0 0 0>}}
  }
  Switcher_Tabs {
   widgetClassPath=&CommonActivatableWidgetSwitcher
   bInherited=true
   bIsVariable=true
   parent=VBox_MainStack
   widget=Switcher_Tabs
   properties {navigation=&self:WidgetTree.Switcher_Tabs.WidgetNavigation_0}
   slot {class=VerticalBoxSlot, properties {padding=<ltrb 8 0 0 8>, size {sizeRule=Fill, value=1}}}
  }
  Widget_Sprays {widgetClassPath=&VerticalBox, bInherited=true, parent=Switcher_Tabs, widget=Widget_Sprays, slot {class=WidgetSwitcherSlot}}
  TexturePicker_Sprays {
   widgetClassPath=/FortniteCoreUI/Customization/AthenaItemTextureVariantPicker
   bInherited=true
   bIsVariable=true
   parent=Widget_Sprays
   widget=TexturePicker_Sprays
   slot {class=VerticalBoxSlot, properties {size {sizeRule=Fill, value=1}}}
  }
  Widget_Settings {
   widgetClassPath=&ScrollBox
   bInherited=true
   parent=Switcher_Tabs
   widget=Widget_Settings
   properties {
    navigationScrollPadding=90
    scrollWhenFocusChanges=AnimatedScroll
    scrollbarPadding=<ltrb 116 2 2 2>
    scrollbarThickness=<12 12>
    widgetBarStyle {
     thickness=16
     draggedThumbImage {
      bIsDynamicallyLoaded=false
      drawAs=RoundedBox
      imageType=FullColor
      mirroring=NoMirror
      resourceName=../../../Engine/Content/Slate/Common/Scrollbar_Thumb.png
      resourceObject=null
      tiling=NoTile
      imageSize=<0 0>
      margin=<ltrb 0.25 0.25 0.25 0.25>
      outlineSettings {
       bUseBrushTransparency=false
       roundingType=FixedRadius
       width=2
       color {colorUseRule=UseColor_Specified, specifiedColor=<rgba 0 0 0 0>}
       cornerRadii=<6 6 6 6>
      }
      tintColor {colorUseRule=UseColor_Specified, specifiedColor=<rgba 1 1 1 1>}
      uVRegion {bIsValid=false, max=<0 0>, min=<0 0>}
     }
     horizontalBackgroundImage {
      bIsDynamicallyLoaded=false
      drawAs=NoDrawType
      imageType=NoImage
      mirroring=NoMirror
      resourceName=None
      resourceObject=null
      tiling=NoTile
      imageSize=<0 0>
      margin=<ltrb 0 0 0 0>
      outlineSettings {
       bUseBrushTransparency=false
       roundingType=HalfHeightRadius
       width=0
       color {colorUseRule=UseColor_Specified, specifiedColor=<rgba 0 0 0 0>}
       cornerRadii=<0 0 0 0>
      }
      tintColor {colorUseRule=UseColor_Specified, specifiedColor=<rgba 1 1 1 1>}
      uVRegion {bIsValid=false, max=<0 0>, min=<0 0>}
     }
     horizontalBottomSlotImage {
      bIsDynamicallyLoaded=false
      drawAs=NoDrawType
      imageType=FullColor
      mirroring=NoMirror
      resourceName=../../../Engine/Content/Slate/Common/Scrollbar_Background_Horizontal.png
      resourceObject=null
      tiling=NoTile
      imageSize=<8 8>
      margin=<ltrb 0 0 0 0>
      outlineSettings {
       bUseBrushTransparency=false
       roundingType=HalfHeightRadius
       width=0
       color {colorUseRule=UseColor_Specified, specifiedColor=<rgba 0 0 0 0>}
       cornerRadii=<0 0 0 0>
      }
      tintColor {colorUseRule=UseColor_Specified, specifiedColor=<rgba 1 1 1 1>}
      uVRegion {bIsValid=false, max=<0 0>, min=<0 0>}
     }
     horizontalTopSlotImage {
      bIsDynamicallyLoaded=false
      drawAs=NoDrawType
      imageType=FullColor
      mirroring=NoMirror
      resourceName=../../../Engine/Content/Slate/Common/Scrollbar_Background_Horizontal.png
      resourceObject=null
      tiling=NoTile
      imageSize=<8 8>
      margin=<ltrb 0 0 0 0>
      outlineSettings {
       bUseBrushTransparency=false
       roundingType=HalfHeightRadius
       width=0
       color {colorUseRule=UseColor_Specified, specifiedColor=<rgba 0 0 0 0>}
       cornerRadii=<0 0 0 0>
      }
      tintColor {colorUseRule=UseColor_Specified, specifiedColor=<rgba 1 1 1 1>}
      uVRegion {bIsValid=false, max=<0 0>, min=<0 0>}
     }
     hoveredThumbImage {
      bIsDynamicallyLoaded=false
      drawAs=RoundedBox
      imageType=FullColor
      mirroring=NoMirror
      resourceName=../../../Engine/Content/Slate/Common/Scrollbar_Thumb.png
      resourceObject=null
      tiling=NoTile
      imageSize=<0 0>
      margin=<ltrb 0.25 0.25 0.25 0.25>
      outlineSettings {
       bUseBrushTransparency=false
       roundingType=FixedRadius
       width=2
       color {colorUseRule=UseColor_Specified, specifiedColor=<rgba 0 0 0 0>}
       cornerRadii=<6 6 6 6>
      }
      tintColor {colorUseRule=UseColor_Specified, specifiedColor=<rgba 1 1 1 1>}
      uVRegion {bIsValid=false, max=<0 0>, min=<0 0>}
     }
     normalThumbImage {
      bIsDynamicallyLoaded=false
      drawAs=RoundedBox
      imageType=FullColor
      mirroring=NoMirror
      resourceName=../../../Engine/Content/Slate/Common/Scrollbar_Thumb.png
      resourceObject=null
      tiling=NoTile
      imageSize=<0 0>
      margin=<ltrb 0.25 0.25 0.25 0.25>
      outlineSettings {
       bUseBrushTransparency=false
       roundingType=FixedRadius
       width=2
       color {colorUseRule=UseColor_Specified, specifiedColor=<rgba 0 0 0 0>}
       cornerRadii=<6 6 6 6>
      }
      tintColor {colorUseRule=UseColor_Specified, specifiedColor=<rgba 1 1 1 0.75>}
      uVRegion {bIsValid=false, max=<0 0>, min=<0 0>}
     }
     verticalBackgroundImage {
      bIsDynamicallyLoaded=false
      drawAs=Box
      imageType=NoImage
      mirroring=NoMirror
      resourceName=None
      resourceObject=/FortUILibrary/UIKit/ScrollBar/MaterialInstances/MI_UIKit_ScrollbarBackground
      tiling=NoTile
      imageSize=<12 12>
      margin=<ltrb 0.5 0.5 0.5 0.5>
      outlineSettings {
       bUseBrushTransparency=false
       roundingType=HalfHeightRadius
       width=0
       color {colorUseRule=UseColor_Specified, specifiedColor=<rgba 0 0 0 0>}
       cornerRadii=<0 0 0 0>
      }
      tintColor {colorUseRule=UseColor_Specified, specifiedColor=<rgba 1 1 1 1>}
      uVRegion {bIsValid=false, max=<0 0>, min=<0 0>}
     }
     verticalBottomSlotImage {
      bIsDynamicallyLoaded=false
      drawAs=NoDrawType
      imageType=FullColor
      mirroring=NoMirror
      resourceName=../../../Engine/Content/Slate/Common/Scrollbar_Background_Vertical.png
      resourceObject=null
      tiling=NoTile
      imageSize=<8 8>
      margin=<ltrb 0 0 0 0>
      outlineSettings {
       bUseBrushTransparency=false
       roundingType=HalfHeightRadius
       width=0
       color {colorUseRule=UseColor_Specified, specifiedColor=<rgba 0 0 0 0>}
       cornerRadii=<0 0 0 0>
      }
      tintColor {colorUseRule=UseColor_Specified, specifiedColor=<rgba 1 1 1 1>}
      uVRegion {bIsValid=false, max=<0 0>, min=<0 0>}
     }
     verticalTopSlotImage {
      bIsDynamicallyLoaded=false
      drawAs=NoDrawType
      imageType=FullColor
      mirroring=NoMirror
      resourceName=../../../Engine/Content/Slate/Common/Scrollbar_Background_Vertical.png
      resourceObject=null
      tiling=NoTile
      imageSize=<8 8>
      margin=<ltrb 0 0 0 0>
      outlineSettings {
       bUseBrushTransparency=false
       roundingType=HalfHeightRadius
       width=0
       color {colorUseRule=UseColor_Specified, specifiedColor=<rgba 0 0 0 0>}
       cornerRadii=<0 0 0 0>
      }
      tintColor {colorUseRule=UseColor_Specified, specifiedColor=<rgba 1 1 1 1>}
      uVRegion {bIsValid=false, max=<0 0>, min=<0 0>}
     }
    }
   }
   slot {class=WidgetSwitcherSlot}
  }
  Setting_Saturation {
   widgetClassPath=&/MutableProjectorSprayCustomizerSetting
   bInherited=true
   bIsVariable=true
   parent=Widget_Settings
   widget=Setting_Saturation
   properties {bIsFocusable=false, settingName=SATURATION}
   slot {class=ScrollBoxSlot}
  }
  Setting_Wear {
   widgetClassPath=&MutableProjectorSprayCustomizerSetting
   bInherited=true
   bIsVariable=true
   parent=Widget_Settings
   widget=Setting_Wear
   properties {bIsFocusable=false, settingName=WEAR}
   slot {class=ScrollBoxSlot, properties {padding=<ltrb 0 16 0 0>}}
  }
  Setting_Scale {
   widgetClassPath=&MutableProjectorSprayCustomizerSetting
   bInherited=true
   bIsVariable=true
   parent=Widget_Settings
   widget=Setting_Scale
   properties {bIsFocusable=false, settingName=SCALE}
   slot {class=ScrollBoxSlot, properties {padding=<ltrb 0 16 0 0>}}
  }
  BottomBarOverlay {
   widgetClassPath=&Overlay
   parent=Locker
   widget=BottomBarOverlay
   slot {class=GridSlot, properties {column=1, columnSpan=2, horizontalAlignment=HAlign_Right, row=4, verticalAlignment=VAlign_Bottom}}
  }
  HorizontalBox_80 {
   widgetClassPath=&HorizontalBox
   parent=BottomBarOverlay
   widget=HorizontalBox_80
   slot {class=OverlaySlot, properties {horizontalAlignment=HAlign_Fill, verticalAlignment=VAlign_Fill}}
  }
  Widget_ArchivedItemsButtons {
   widgetClassPath=&HorizontalBox
   bInherited=true
   parent=HorizontalBox_80
   widget=Widget_ArchivedItemsButtons
   slot {class=HorizontalBoxSlot}
  }
  Button_ShowArchivedItems {
   widgetClassPath=/FortUILibrary/UIKit/Buttons/WBP_UIKit_Button_Quiet
   bInherited=true
   bIsVariable=true
   parent=Widget_ArchivedItemsButtons
   widget=Button_ShowArchivedItems
   properties {text=Show Archived Items, triggeringInputAction {dataTable=/FortUILibrary/Input/LockerInputActionsDataTable, rowName=ShowArchived}}
   slot {class=HorizontalBoxSlot, properties {padding=<ltrb 24 0 0 0>}}
  }
  Button_HideArchivedItems {
   widgetClassPath=&WBP_UIKit_Button_Quiet
   bInherited=true
   bIsVariable=true
   parent=Widget_ArchivedItemsButtons
   widget=Button_HideArchivedItems
   properties {text=Hide Archived Items, triggeringInputAction {dataTable=&LockerInputActionsDataTable, rowName=HideArchived}}
   slot {class=HorizontalBoxSlot, properties {padding=<ltrb 24 0 0 0>}}
  }
  Button_Camera {
   widgetClassPath=&WBP_UIKit_Button_Quiet
   bInherited=true
   bIsVariable=true
   parent=HorizontalBox_80
   widget=Button_Camera
   properties {text=Camera, triggeringInputAction {dataTable=/FortUILibrary/Input/ItemCameraInputActionDataTable, rowName=CanControlCamera}}
   slot {class=HorizontalBoxSlot, properties {padding=<ltrb 24 0 0 0>}}
  }
  Button_Confirm {
   widgetClassPath=&WBP_UIKit_Button_Quiet
   bInherited=true
   bIsVariable=true
   parent=HorizontalBox_80
   widget=Button_Confirm
   properties {text=Accept, triggeringInputAction {dataTable=&InputActionDataTable, rowName=GenericAccept}}
   slot {class=HorizontalBoxSlot, properties {verticalAlignment=VAlign_Top, padding=<ltrb 24 0 0 0>}}
  }
  VisibilityBorder_NonTouch {
   widgetClassPath=&CommonHardwareVisibilityBorder
   parent=HorizontalBox_80
   widget=VisibilityBorder_NonTouch
   properties {
    padding=<ltrb 24 0 0 0>
    visibilityQuery {
     autoDescription=" NONE( Input.Touch )"
     tokenStreamVersion=0
     userDescription=""
     queryTokenStream [0,1,3,1,0]
     tagDictionary [{tagName=Input.Touch}]
    }
   }
   slot {class=HorizontalBoxSlot}
  }
  Button_Back {
   widgetClassPath=&WBP_UIKit_Button_Quiet
   bInherited=true
   bIsVariable=true
   parent=VisibilityBorder_NonTouch
   widget=Button_Back
   properties {text=Back, triggeringInputAction {dataTable=&InputActionDataTable, rowName=GenericBack}}
   slot {class=BorderSlot, properties {padding=<ltrb 24 0 0 0>}}
  }
  WarningMessage {
   widgetClassPath=/FortUILibrary/UIKit/InfoBox/WBP_UIKit_InfoBox
   bIsVariable=true
   parent=Locker
   widget=WarningMessage
   properties {visibility=Collapsed, iconSize=<64 64>, padding=<ltrb 64 0 0 0>}
   slot {class=GridSlot, properties {columnSpan=2, row=3, verticalAlignment=VAlign_Center}}
  }
  Widget_CameraFrame {
   widgetClassPath=&SizeBox
   bInherited=true
   bIsVariable=true
   parent=GridPanel
   widget=Widget_CameraFrame
   slot {class=GridSlot, properties {horizontalAlignment=HAlign_Left}}
  }
  HorizontalBox {widgetClassPath=&HorizontalBox, parent=Widget_CameraFrame, widget=HorizontalBox, slot {class=SizeBoxSlot}}
  Camera_Spacer_Left {
   widgetClassPath=&Spacer
   bIsVariable=true
   parent=HorizontalBox
   widget=Camera_Spacer_Left
   slot {class=HorizontalBoxSlot, properties {size {sizeRule=Fill, value=0}}}
  }
  SizeBox {
   widgetClassPath=&SizeBox
   parent=HorizontalBox
   widget=SizeBox
   properties {bOverride_HeightOverride=true, bOverride_WidthOverride=true, heightOverride=1024, widthOverride=1024}
   slot {class=HorizontalBoxSlot}
  }
  FortItemPreviewWidget {
   widgetClassPath=/FortniteCoreUI/Frontend/ItemPreview/WBP_ItemPreview
   bIsVariable=true
   parent=SizeBox
   widget=FortItemPreviewWidget
   properties {
    bPreviewActionsEnabled=false
    bSupportsActivationFocus=false
    itemViewObject=&self:WidgetTree.FortItemPreviewWidget.ItemViewObject
    screenTag {tagName=UI.ItemPreview.Screen.Locker.VariantSelector}
   }
   slot {class=SizeBoxSlot}
  }
  Camera_Spacer_Right {
   widgetClassPath=&Spacer
   bIsVariable=true
   parent=HorizontalBox
   widget=Camera_Spacer_Right
   slot {class=HorizontalBoxSlot, properties {size {sizeRule=Fill, value=1}}}
  }
  Disclaimer_Container {
   widgetClassPath=&HorizontalBox
   bIsVariable=true
   parent=GridPanel
   widget=Disclaimer_Container
   slot {class=GridSlot, properties {verticalAlignment=VAlign_Bottom}}
  }
  Image_0 {
   widgetClassPath=&Image
   parent=Disclaimer_Container
   widget=Image_0
   properties {
    visibility=SelfHitTestInvisible
    brush {
     bIsDynamicallyLoaded=false
     drawAs=Image
     imageType=NoImage
     mirroring=NoMirror
     resourceName=None
     resourceObject=/Game/UI/Foundation/Textures/Icons/Heroes/Variants/T-Icon-Variant-32
     tiling=NoTile
     imageSize=<32 32>
     margin=<ltrb 0 0 0 0>
     outlineSettings {
      bUseBrushTransparency=false
      roundingType=HalfHeightRadius
      width=0
      color {colorUseRule=UseColor_Specified, specifiedColor=<rgba 0 0 0 0>}
      cornerRadii=<0 0 0 0>
     }
     tintColor {colorUseRule=UseColor_Specified, specifiedColor=<rgba 1 1 1 1>}
     uVRegion {bIsValid=false, max=<0 0>, min=<0 0>}
    }
   }
   slot {class=HorizontalBoxSlot, properties {horizontalAlignment=HAlign_Center, verticalAlignment=VAlign_Center}}
  }
  Spacer_0 {widgetClassPath=&Spacer, parent=Disclaimer_Container, widget=Spacer_0, properties {size=<8 1>}, slot {class=HorizontalBoxSlot}}
  Text_StyleDisclaimer {
   widgetClassPath=&CommonTextBlock
   bIsVariable=true
   parent=Disclaimer_Container
   widget=Text_StyleDisclaimer
   properties {
    justification=Center
    renderOpacity=0.5
    style=/FortUILibrary/Styles/Typography/_TextStyle_Legal._TextStyle_Legal_C
    text=All styles shown here are included with purchase and can be changed at any time in the Locker.
    font {
     bForceMonospaced=false
     bMaterialIsStencil=false
     fontMaterial=null
     fontObject=&NotoSans
     letterSpacing=0
     monospacedWidth=1
     size=15
     skewAmount=0
     typefaceFontName=NotoSans-Regular
     outlineSettings {
      bApplyOutlineToDropShadows=false
      bMiteredCorners=false
      bSeparateFillAlpha=false
      outlineMaterial=null
      outlineSize=0
      outlineColor=<rgba 0 0 0 1>
     }
    }
    shadowOffset=<0 0>
    strikeBrush {
     bIsDynamicallyLoaded=false
     drawAs=NoDrawType
     imageType=NoImage
     mirroring=NoMirror
     resourceName=None
     resourceObject=null
     tiling=NoTile
     imageSize=<32 32>
     margin=<ltrb 0 0 0 0>
     outlineSettings {
      bUseBrushTransparency=false
      roundingType=HalfHeightRadius
      width=0
      color {colorUseRule=UseColor_Specified, specifiedColor=<rgba 0 0 0 0>}
      cornerRadii=<0 0 0 0>
     }
     tintColor {colorUseRule=UseColor_Specified, specifiedColor=<rgba 1 1 1 1>}
     uVRegion {bIsValid=false, max=<0 0>, min=<0 0>}
    }
   }
   slot {class=HorizontalBoxSlot, properties {padding=<ltrb 16 16 16 16>}}
  }
  Button_PreviewAction {widgetClassPath=&CommonButtonBase, bInherited=true, widget=null}
  Button_PreviewAction1 {widgetClassPath=&CommonButtonBase, bInherited=true, widget=null}
  ItemDownloadIndicator {widgetClassPath=/Script/FortniteUI.FortItemDownloadIndicator, bInherited=true, widget=null}
  Text_DebugPreviewDetails {widgetClassPath=&CommonTextBlock, bInherited=true, widget=null}
 ]
}
